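# Translucent, two-sided, unlit material for the Modular Test Instance testing tool, likely its perimeter, showing a glowing hexagon pattern.
# A world-aligned hexagon texture blends emissive color between violet-blue and cyan, and opacity between min and max parameters, through material attributes.
# Time-scrolled hex noise, a four-color lerp and two HexPulse samples are present but unconnected to the output, so the visible pattern is static.
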Material /Game/NeverCook/EditorUtilities/TestingTools/ModularTestInstance/Art/M_MTI_Perim
graph {
 properties {
  bCastRayTracedShadows=false
  bUseMaterialAttributes=true
  blendMode=BLEND_Translucent
  pixelDepthOffsetMode=PDOM_Legacy
  shadingModel=MSM_Unlit
  twoSided=true
  usedShadingModels=Unlit
 }
 nodes [
  n1 = Multiply(A=Time(), B=ScalarParameter() {defaultValue=0.1, parameterName=TimeMult})
  n2 = MaterialFunctionCall(
   TextureObject=TextureObject() {texture=/Game/Effects/Fort_Effects/Textures/Tile/T_Hex_Noise}
   TextureSize=ScalarParameter() {defaultValue=500, parameterName=Texturesize}
  ) {materialFunction=/Game/Weapons/WeaponSkins/Wraps/MaterialFunctions/WorldAlignedTexture}
  Frac(None=Add(A=@n1, B=@n2.XYZ Texture))
  n3 = Constant3Vector() {constant=<rgba 0.2444 0.1655 1 1>}
  n4 = Constant3Vector() {constant=<rgba 0.0235 0.8198 1 1>}
  MaterialFunctionCall(
   1=@n3
   2=@n4
   3=Constant3Vector() {constant=<rgba 1 0.1511 0.567 1>}
   4=Constant3Vector() {constant=<rgba 1 0.99 0.1639 1>}
   A=Fmod(A=@n1, B=@n2.XYZ Texture)
  ) {materialFunction=&Lerp_Multiple_Float3}
  n5 = TextureObject() {
   samplerType=SAMPLERTYPE_LinearColor
   texture=/Game/Environments/Artemis/Props/Quest/Quest_TionfoilHoloTable_A/Textures/T_Tinfoil_HoloTable_Hexagons
  }
  n6 = MaterialFunctionCall(TextureObject=@n5, TextureSize=ScalarParameter() {defaultValue=500, parameterName=Texturesize}) {materialFunction=&WorldAlignedTexture}
  SetMaterialAttributes!(
   MaterialAttributes
   Emissive Color=LinearInterpolate(A=@n3, B=@n4, Alpha=@n6.XYZ Texture)
   Opacity=LinearInterpolate(
    A=ScalarParameter() {defaultValue=0.1, parameterName=OpacityMin}
    B=ScalarParameter() {defaultValue=1, parameterName=OpacityMax}
    Alpha=@n6.XYZ Texture
   )
  ) {attributeSetTypes [#1, #2]}
  TextureSample() {texture=/Game/Creative/Sets/PropSets/Primitives/Customization/Textures/Hex/T_CM_HexPulse_M}
  TextureSample() {samplerType=SAMPLERTYPE_Masks, texture=/Game/Creative/Sets/PropSets/Primitives/Customization/Textures/Hex/T_CM_HexPulse_RandomColor}
 ]
}
